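# Niagara swing-trail effect for the ChiveFlake Keytar cosmetic, drawn as two layered ribbons.
# MainTrail and a narrower, shorter-lived SupportTrail use the FN AnimTrail modules to spawn, initialize and update ribbons.
# System-level TOD and TrailFade values set spawn color and alpha, and lifetimes shorten as the system loop ages.
# Curves tint and fade each ribbon over its life; velocity and alternate time-of-day color modules are disabled.
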NiagaraSystem /SparksCosmetics/Instrument/Keytar/ChiveFlake/Effects/NS_Keytar_ChiveFlake_Swing
dependencies {
 usedDataInterfaces [&NiagaraDataInterfaceColorCurve, &NiagaraDataInterfaceCurve, &NiagaraDataInterfaceSocketReader]
 usedDynamicInputs [&MergeHelper_FloatSwitch]
 usedModules [
  &self:SystemSpawnScript.SystemScriptSource.SystemScriptGraph.NiagaraNodeAssignment_0.SetVariables_#1
  &SystemState
  &self:SystemSpawnScript.SystemScriptSource.SystemScriptGraph.NiagaraNodeAssignment_2.SetVariables_#2
 ]
}
emitters {
 MainTrail {
  emitterUpdateScript [
   &EmitterState()
   &FN_SpawnAnimTrail(Spawn Interval=0) {module=FN_SpawnAnimTrail001}
  ]
  particleSpawnScript [
   &FN_InitializeAnimTrailBoneIndecies()
   &FN_InitializeAnimTrail(
    AnimTrailWidth=1.02
    Color=&MultiplyLinearColor(
     Linear Color=$System.TOD
     Scale Factor=&MakeFloat4FromVectorAndFloat("Vector (RGB)"=<1 1 1>, "Float (Alpha)"=$System.TrailFade)
    )
    Lifetime=&Lerp_Float(A=0.13, B=0.05, Alpha=$System.NormalizedLoopAge)
   )
   &AddVelocity(Velocity=<750 0 0>) {enabled=false}
  ]
  particleUpdateScript [
   &SolveForcesAndVelocity(Clamp Velocity=false, Limit Acceleration=false) {enabled=false}
   &ScaleColor(
    Scale Alpha=&FloatFromCurve(FloatCurve=DataInterface() {propertyValues {Curve {keys [{arriveTangent=-3.8864, interpMode=RCIM_Cubic, time=0.3, value=1}, <at 1 0 RCIM_Cubic>]}}})
    Scale RGB=&MakeVectorFromLinearColorRGB(Color=&ColorFromCurve(ColorCurve=DataInterface() {propertyValues {
        AlphaCurve {keys [<at 0 1>]}
        BlueCurve {keys [<at 0.017 0.9652>,<at 0.6986 0.8696>]}
        GreenCurve {keys [<at 0.017 0.6211>,<at 0.6986 0.3629>]}
        RedCurve {keys [<at 0.017 0.7129>,<at 0.6986 0.498>]}
       }}))
    ScaleA=true
    ScaleRGB=true
   )
   &ScaleColor(
    Scale Alpha=&FloatFromCurve(FloatCurve=DataInterface() {propertyValues {Curve {keys [{arriveTangent=-3.8864, interpMode=RCIM_Cubic, time=0.3, value=1}, <at 1 0 RCIM_Cubic>]}}})
    Scale RGB=&Multiply_Vector(
     A=&MakeVectorFromLinearColorRGB(Color=&ColorFromCurve(ColorCurve=DataInterface() {propertyValues {
         AlphaCurve {keys [<at 0 1>]}
         BlueCurve {keys [<at 0.017 0.9652>,<at 0.6986 0.8696>]}
         GreenCurve {keys [<at 0.017 0.6211>,<at 0.6986 0.3629>]}
         RedCurve {keys [<at 0.017 0.7129>,<at 0.6986 0.498>]}
        }}))
     B=&MakeVectorFromLinearColorRGB(Color=&FN_Select_Color_from_Time_Of_Day(
       Use Manual Time of Day Thresholds=true
       Morning=<rgba 0.493 0.4792 0.7812 1>
       Day=<rgba 1 1 1 1>
       Evening=<rgba 0.5469 0.5284 0.7865 1>
       Night=<rgba 0.3581 0.3594 0.7812 1>
       Morning Threshold=9
       Day Threshold=17
       Evening Threshold=21
      ))
    )
    ScaleA=true
    ScaleRGB=true
   ) {enabled=false, module=ScaleColor001}
   &FN_UpdateAnimTrail() {module=FN_UpdateAnimTrail001}
  ]
  rendererClasses [&NiagaraRibbonRendererProperties]
  renderers [{rendererClass=&NiagaraRibbonRendererProperties}]
 }
 SupportTrail {
  emitterUpdateScript [
   &EmitterState()
   &FN_SpawnAnimTrail(Spawn Interval=0)
  ]
  particleSpawnScript [
   &FN_InitializeAnimTrailBoneIndecies()
   &FN_InitializeAnimTrail(
    AnimTrailWidth=0.75
    Color=&MultiplyLinearColor(
     Linear Color=&MultiplyLinearColor(
      Linear Color=<rgba 2.8 1.4146 1.5412 0.5>
      Scale Factor=&MakeVector4_FromLinearColor(LinearColor=$System.TOD)
     )
     Scale Factor=&MakeVector4(X=1, Y=1, Z=1, W=$System.TrailFade)
    )
    Lifetime=&Lerp_Float(A=0.11, B=0.05, Alpha=$System.NormalizedLoopAge)
   )
   &AddVelocity(Velocity=<750 0 0>) {enabled=false}
  ]
  particleUpdateScript [
   &SolveForcesAndVelocity(Clamp Velocity=false, Limit Acceleration=false) {enabled=false}
   &FN_UpdateAnimTrail() {module=FN_UpdateAnimTrail001}
   &ScaleColor(
    Scale Alpha=&FloatFromCurve(
     FloatCurve=DataInterface() {propertyValues {Curve {keys [
         {arriveTangent=-2.003, interpMode=RCIM_Cubic, leaveTangent=-2.003, tangentMode=RCTM_User, time=0.3, value=1}
         {interpMode=RCIM_Cubic, tangentMode=RCTM_User, time=0.5, value=0}
        ]}}}
     Scale Curve=1.5
    )
    Scale RGB=&MakeVectorFromLinearColorRGB(Color=&ColorFromCurve(ColorCurve=DataInterface() {propertyValues {
        AlphaCurve {keys []}
        BlueCurve {keys [<at 0 2>,<at 0.0676 1>]}
        GreenCurve {keys [<at 0 2>,<at 0.0676 1>]}
        RedCurve {keys [<at 0 2>,<at 0.0676 1>]}
       }}))
    ScaleA=true
    ScaleRGB=true
   )
   &ScaleColor(
    Scale Alpha=&FloatFromCurve(
     FloatCurve=DataInterface() {propertyValues {Curve {keys [
         {arriveTangent=-2.003, interpMode=RCIM_Cubic, leaveTangent=-2.003, tangentMode=RCTM_User, time=0.3, value=1}
         {interpMode=RCIM_Cubic, tangentMode=RCTM_User, time=0.5, value=0}
        ]}}}
     Scale Curve=1.5
    )
    Scale RGB=&Multiply_Vector(
     A=&MakeVectorFromLinearColorRGB(Color=&ColorFromCurve(ColorCurve=DataInterface() {propertyValues {
         AlphaCurve {keys []}
         BlueCurve {keys [<at 0 2>,<at 0.0676 1>,<at 0.6157 0.2074>]}
         GreenCurve {keys [<at 0 2>,<at 0.0676 1>,<at 0.6157 0.1319>]}
         RedCurve {keys [<at 0 2>,<at 0.0676 1>,<at 0.6157 0.6562>]}
        }}))
     B=&MakeVectorFromLinearColorRGB(Color=&ColorFromCurve(
       ColorCurve=DataInterface() {propertyValues {
         AlphaCurve {keys [<at 0 1>,<at 0.125 1>,<at 0.25 1>,<at 0.375 1>,<at 0.5 1>,<at 0.625 1>,<at 0.75 1>,<at 0.875 1>,<at 1 1>]}
         BlueCurve {keys [<at 0.125 0.7835>,<at 0.166 0.6172>,<at 0.208 0.5395>,<at 0.25 0.6172>,<at 0.291 0.6939>,<at 0.333 0.7379>,<at 0.375 0.7835>,<at 0.416 0.7991>,<at 0.458 0.8148>,<at 0.5 0.8308>,<at 0.541 0.8148>,<at 0.583 0.7991>,<at 0.625 0.7991>,<at 0.666 0.7529>,<at 0.708 0.7084>,<at 0.75 0.5711>,<at 0.791 0.4735>,<at 0.833 0.3231>,<at 0.8778 0.8388>]}
         GreenCurve {keys [<at 0.125 0.5583>,<at 0.166 0.5333>,<at 0.208 0.5271>,<at 0.25 0.6376>,<at 0.291 0.7529>,<at 0.333 0.7835>,<at 0.375 0.8148>,<at 0.416 0.7991>,<at 0.458 0.8148>,<at 0.5 0.8469>,<at 0.541 0.7991>,<at 0.583 0.7991>,<at 0.625 0.8148>,<at 0.666 0.7991>,<at 0.708 0.7529>,<at 0.75 0.5711>,<at 0.791 0.5029>,<at 0.833 0.5029>,<at 0.8778 0.5333>]}
         RedCurve {keys [<at 0.125 0.2831>,<at 0.166 0.7084>,<at 0.208 0.6867>,<at 0.25 0.8148>,<at 0.291 0.8879>,<at 0.333 0.8963>,<at 0.375 0.9131>,<at 0.416 0.8714>,<at 0.458 0.8714>,<at 0.5 0.9131>,<at 0.541 0.8632>,<at 0.583 0.8714>,<at 0.625 0.9131>,<at 0.666 0.8963>,<at 0.708 0.8714>,<at 0.75 0.7529>,<at 0.791 0.7529>,<at 0.833 0.8469>,<at 0.8778 0.1878>]}
        }}
       CurveIndex=Linked() {linkedVariable {name=NPC.FortniteNPC.TimeOfDay_0-1, type=NiagaraFloat}}
      ))
    )
    ScaleA=true
    ScaleRGB=true
   ) {enabled=false, module=ScaleColor001}
  ]
  rendererClasses [&NiagaraRibbonRendererProperties]
  renderers [{rendererClass=&NiagaraRibbonRendererProperties}]
 }
}
summary {
 systemName=NS_Keytar_ChiveFlake_Swing
 emitters [
  {emitterName=MainTrail, rendererClasses [&NiagaraRibbonRendererProperties]}
  {emitterName=SupportTrail, rendererClasses [&NiagaraRibbonRendererProperties]}
 ]
 userVariables [
  User.RibbonIDOffset {type=NiagaraInt32}
  User.SwingSectionIndex {type=NiagaraInt32}
 ]
}
