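# Material function that draws the ChiveFlake keytar's animated, music-reactive face over two input colors.
# Three flipbook layers (mouth, pulsing star, eyes) use a shared flipbook function and composite by mask, with static switches for star and eyes.
# Curve atlas rows, timed from game-time parameters, supply frame indices and offsets for idle, blink and breaker animations.
# The music clock parameter collection provides beat, tempo and music state for pulsing and head bump.
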MaterialFunction /SparksCosmetics/Instrument/Keytar/ChiveFlake/Materials/MF_ChiveFlake_Keytar
graph {nodes [
  Add(B=OneMinus(None=NamedRerouteUsage()))
  MaterialExpressionMultiply_34 = Multiply(A=Time(), B=ScalarParameter() {defaultValue=0.25, group=UpbeatIguana, parameterName=UI_BackpackBlinkSpeed}) {constB=0.25}
  n1 = Custom!(t=@MaterialExpressionMultiply_34) {
   code="return frac(2*sin(0.5*t) + 0.15*sin(4.5*t));"
   description=RandomBlink
   showCode=true
   inputs [{inputName=t, input {expression=&self:MaterialExpressionMultiply_34}}]
  }
  n2 = Multiply(A=ComponentMask(None=@n1) {r=true}) {constB=10}
  AppendVector(A=@n2, B=Saturate(None=@n2))
  Constant() {r=3}
  n3 = Modulo!(A=Divide(A=Time() {period=1}) {constB=8}, B=Constant() {r=1})
  n4 = Multiply(A=Abs(None=Subtract(A=@n3) {constB=0.5})) {constB=7}
  n5 = Divide(A=Time() {period=1}, B=Add(A=Floor(None=@n4)) {constB=20}) {constB=8}
  n6 = Constant() {r=1}
  n7 = Abs(None=Subtract(A=Modulo(A=@n5, B=@n6)) {constB=0.5})
  Divide(A=Floor(None=Multiply(A=@n7) {constB=7})) {constB=7}
  Divide(A=Time()) {constB=2048}
  n8 = If(
   A=ScalarParameter() {defaultValue=60, group=Flipbook General, parameterName=BPM}
   B=ScalarParameter() {defaultValue=120, group=Flipbook General, parameterName=HighBPMTreshold}
   "A > B"=NamedRerouteUsage()
   "A < B"=NamedRerouteUsage()
  ) {constB=120}
  n9 = Constant() {r=15}
  Divide(A=Floor(None=Multiply(A=@n8, B=@n9) {constB=30}), B=@n9) {constB=30}
  Fmod(A=Constant(), B=Constant() {r=2})
  n10 = LinearInterpolate(A=FunctionInput() {previewValue=<0.7011 0.3515 1 1>}, B=NamedRerouteUsage(), Alpha=NamedRerouteUsage())
  n11 = LinearInterpolate(A=@n10, B=NamedRerouteUsage(), Alpha=NamedRerouteUsage())
  n12 = NamedRerouteUsage()
  n13 = NamedRerouteUsage()
  n14 = StaticSwitchParameter(
   True=StaticSwitchParameter(True=LinearInterpolate(A=@n11, B=@n12, Alpha=@n13), False=@n11) {defaultValue=true, group=Flipbook C, parameterName=Use Flipbook C, sortPriority=0}
   False=@n10
  ) {defaultValue=true, group=Flipbook B, parameterName=Use Flipbook B, sortPriority=0}
  FunctionOutput(None=@n14)
  n15 = LinearInterpolate(A=FunctionInput() {inputName=In_1, previewValue=<0.7011 0.3515 1 1>}, B=NamedRerouteUsage(), Alpha=NamedRerouteUsage())
  n16 = LinearInterpolate(A=@n15, B=NamedRerouteUsage(), Alpha=NamedRerouteUsage())
  n17 = NamedRerouteUsage()
  n18 = NamedRerouteUsage()
  n19 = StaticSwitchParameter(
   True=StaticSwitchParameter(True=LinearInterpolate(A=@n16, B=@n17, Alpha=@n18), False=@n16) {defaultValue=true, group=Flipbook C, parameterName=Use Flipbook C, sortPriority=0}
   False=@n15
  ) {defaultValue=true, group=Flipbook B, parameterName=Use Flipbook B, sortPriority=0}
  FunctionOutput(None=@n19) {outputName=Result_1}
  n20 = NamedRerouteUsage()
  n21 = Multiply(
   A=ScalarParameter() {defaultValue=4, group=Flipbook A, parameterName=A_Number of Columns, sortPriority=4}
   B=ScalarParameter() {defaultValue=4, group=Flipbook A, parameterName=A_Number of Rows, sortPriority=5}
  )
  n22 = Multiply(A=@n21) {constB=2}
  If(A=@n20, B=Subtract(A=@n21), "A > B"=If(A=@n20, B=Subtract(A=@n22)))
  n23 = NamedRerouteUsage()
  n24 = CurveAtlasRowParameter(CurveTime=@n23) {
   atlas=/SparksCosmetics/Instrument/Keytar/ChiveFlake/Animation/CA_ChiveFlake_Keytar_Animations
   curve=/SparksCosmetics/Instrument/Keytar/ChiveFlake/Animation/CLC_ChiveFlake_Keytar_BeatChomp
   defaultValue=12
   group=Flipbook General
   parameterName=AnimMusic_CurveAtlas
  }
  LinearInterpolate(B=@n24, Alpha=NamedRerouteUsage()) {constAlpha=1}
  n25 = MaterialFunctionCall!(Vector3=NamedRerouteUsage(), MaximumNumberOfDigits, UVs, DebugTextLocation RG_UpperRight BA_LowerLeft, Component Spacing) {materialFunction=/Engine/Functions/Engine_MaterialFunctions02/Utility/DebugFloat3Values}
  LinearInterpolate(A=@n19, B=@n25.ColorCodedOutput, Alpha=@n25.ColorCodedOutput)
  LinearInterpolate(B=Floor(None=@n23), Alpha=NamedRerouteUsage()) {constAlpha=1}
  LinearInterpolate(A=@n14, B=@n25.ColorCodedOutput, Alpha=@n25.ColorCodedOutput)
  LinearInterpolate(B=NamedRerouteUsage())
  LinearInterpolate() {constB=0}
  MaterialFunctionCall!(Vector2=Multiply(A=Time()) {constB=0.9}, MaximumNumberOfDigits, UVs, DebugTextLocation RG_UpperRight BA_LowerLeft, Component Spacing) {materialFunction=/Engine/Functions/Engine_MaterialFunctions02/Utility/DebugFloat2Values}
  n26 = TextureCoordinate()
  n27 = AppendVector(A=Constant(), B=Constant() {r=-0.3})
  MaterialFunctionCall(Number=NamedRerouteUsage(), UVs=Multiply(A=Add(A=@n26, B=@n27)) {constB=2}) {materialFunction=&DebugScalarValues}
  n28 = NamedRerouteUsage()
  n29 = ComponentMask(None=@n28) {a=true}
  n30 = NamedRerouteUsage()
  NamedRerouteDeclaration(None=MaterialFunctionCall(In=Multiply(A=@n29, B=@n30), Power=Constant() {r=3}) {materialFunction=&SCurve}.Result) {name=A_FlipbookMask, nodeColor=<rgba 1 0.375 0 1>}
  NamedRerouteDeclaration(None=ComponentMask(None=@n28) {b=true, g=true, r=true}) {name=A_FlipbookColor, nodeColor=<rgba 0.0547 0 1 1>}
  n31 = Subtract(A=CollectionParameter() {collection=/Game/Sounds/MusicEventSubsystem/MPC_MusicEventSubsystem_PriorityClock, parameterName=MeasureBeat}) {constB=-0.1}
  n32 = If(A=Constant() {r=107}, B=NamedRerouteUsage(), "A > B"=Constant() {r=1}, "A < B"=Constant() {r=2}) {constB=60}
  NamedRerouteDeclaration(None=Saturate(None=Fmod(A=@n31, B=@n32))) {name=Pulse, nodeColor=<rgba 0.0078 0 1 1>}
  n33 = NamedRerouteUsage()
  n34 = ScalarParameter() {defaultValue=4, group=Flipbook C, parameterName=C_GridSize, sortPriority=4}
  n35 = MaterialFunctionCall!(
   OffsetScale=ScalarParameter() {defaultValue=0.65, group=Flipbook A, parameterName=A_OffsetScale, sortPriority=2}
   OffsetRotation=ScalarParameter() {group=Flipbook A, parameterName=A_OffsetRotation, sortPriority=2}
   OffsetU=Add(A=ScalarParameter() {group=Flipbook A, parameterName=A_OffsetU, sortPriority=2}, B=ComponentMask(None=@n33) {g=true}) {constB=0}
   OffsetV=ScalarParameter() {defaultValue=0.1, group=Flipbook A, parameterName=A_OffsetV, sortPriority=3}
   Number of Columns=@n34
   Index=ComponentMask(None=NamedRerouteUsage()) {g=true}
   Number of Rows=@n34
  ) {materialFunction=&/MF_ChiveFlake_Keytar_Flipbook}
  NamedRerouteDeclaration(None=@n35.Coordinates) {name=MouthCoordinates, nodeColor=<rgba 1 0 0.7266 1>}
  n36 = Subtract(A=NamedRerouteUsage(), B=ScalarParameter() {defaultValue=-10, group=Flipbook General, parameterName=BlinkGameTime})
  n37 = CurveAtlasRowParameter(CurveTime=Constant() {r=1}) {
   atlas=&CA_ChiveFlake_Keytar_Animations
   curve=/SparksCosmetics/Instrument/Keytar/ChiveFlake/Animation/CLC_ChiveFlake_Keytar_Idle_01
   defaultValue=1
   group=Flipbook General
   parameterName=Idle_CurveAtlas
  }
  n38 = CurveAtlasRowParameter(CurveTime=Divide(A=@n36, B=@n37.A) {constB=0.5, desc=time it lasts}) {
   atlas=&CA_ChiveFlake_Keytar_Animations
   curve=&CLC_ChiveFlake_Keytar_Idle_01
   defaultValue=1
   group=Flipbook General
   parameterName=Idle_CurveAtlas
  }
  n39 = CurveAtlasRowParameter() {
   atlas=&CA_ChiveFlake_Keytar_Animations
   curve=&CLC_ChiveFlake_Keytar_Idle_01
   defaultValue=1
   group=Flipbook General
   parameterName=Idle_CurveAtlas
  }
  n40 = Constant()
  n41 = Constant() {r=1}
  n42 = AppendVector(A=LinearInterpolate(A=@n38.B, B=@n39.R, Alpha=If(A=@n38.A, B=@n36, "A > B"=@n40, "A < B"=@n41)) {constAlpha=0}, B=@n39.G)
  n43 = CurveAtlasRowParameter() {
   atlas=&CA_ChiveFlake_Keytar_Animations
   curve=/SparksCosmetics/Instrument/Keytar/ChiveFlake/Animation/CLC_ChiveFlake_Keytar_Idle_03
   defaultValue=3
   group=Flipbook General
   parameterName=IdleMusic_CurveAtlas
  }
  NamedRerouteDeclaration(None=LinearInterpolate(A=AppendVector(A=@n42, B=@n39.B), B=@n43, Alpha=NamedRerouteUsage())) {name=Idle, nodeColor=<rgba 1 0 0.4688 1>}
  n44 = NamedRerouteUsage()
  n45 = ScalarParameter() {defaultValue=-10, group=Flipbook General, parameterName=PulseGameTime}
  n46 = CurveAtlasRowParameter(CurveTime=LinearInterpolate(A=Subtract(A=@n44, B=@n45), B=NamedRerouteUsage(), Alpha=NamedRerouteUsage())) {
   atlas=&CA_ChiveFlake_Keytar_Animations
   curve=/SparksCosmetics/Instrument/Keytar/ChiveFlake/Animation/CLC_ChiveFlake_Keytar_StarPulsing
   group=Flipbook General
   parameterName=Pulse_CurveAtlas
  }
  n47 = Divide(A=ScalarParameter() {defaultValue=1, group=Flipbook General, parameterName=HeadBumpIntensity}) {constB=100}
  NamedRerouteDeclaration(None=LinearInterpolate(B=Multiply(A=@n46, B=@n47) {constB=0.01}, Alpha=ScalarParameter() {defaultValue=1, group=Flipbook General, parameterName=UseHeadBumpToBeat}) {constB=0}) {name=FaceBeatOffset, nodeColor=<rgba 0.0312 0 1 1>}
  n48 = CurveAtlasRowParameter() {
   atlas=&CA_ChiveFlake_Keytar_Animations
   curve=/SparksCosmetics/Instrument/Keytar/ChiveFlake/Animation/CLC_ChiveFlake_Keytar_Mad
   defaultValue=9
   group=Flipbook General
   parameterName=Breaker_CurveAtlas
  }
  n49 = Divide(A=Subtract(A=NamedRerouteUsage(), B=ScalarParameter() {defaultValue=-10, group=Flipbook General, parameterName=BreakerGameTime}), B=@n48.A) {constB=5, desc=time it lasts}
  n50 = Saturate(None=@n49)
  n51 = ScalarParameter() {group=Flipbook General, parameterName=BreakerFreeze, sliderMax=1}
  n52 = ScalarParameter() {group=Flipbook General, parameterName=BreakerLoop, sliderMax=1}
  n53 = Saturate(None=Subtract(A=Floor(None=@n50), B=Add(A=@n51, B=@n52)))
  NamedRerouteDeclaration(None=OneMinus(None=@n53)) {name=IsPlayingAnimaiton, nodeColor=<rgba 0.8516 1 0 1>}
  NamedRerouteDeclaration(None=CollectionParameter() {collection=&MPC_MusicEventSubsystem_PriorityClock, parameterName=MusicTimestampValid}) {name=IsMusicOn, nodeColor=<rgba 1 0.0408 0 1>}
  NamedRerouteDeclaration(None=CollectionParameter() {collection=&MPC_MusicEventSubsystem_PriorityClock, parameterName=Tempo}) {name=BPM, nodeColor=<rgba 1 0.7734 0 1>}
  NamedRerouteDeclaration(None=StaticSwitch(True=ScalarParameter() {parameterName=LocalTime}, False=Time() {period=7})) {name=Time, nodeColor=<rgba 0.3372 0.0176 0.0168 1>}
  NamedRerouteDeclaration() {name=DebugValue, nodeColor=<rgba 0 0 0 1>}
  n54 = NamedRerouteUsage()
  n55 = Add(A=NamedRerouteUsage(), B=@n54)
  n56 = StaticSwitchParameter(True=Add(A=NamedRerouteUsage(), B=@n55), False=@n55) {defaultValue=true, group=Flipbook C, parameterName=Use Flipbook C, sortPriority=0}
  NamedRerouteDeclaration(None=Clamp(None=StaticSwitchParameter(True=@n56, False=@n54) {defaultValue=true, group=Flipbook B, parameterName=Use Flipbook B, sortPriority=0})) {name=CombinedMasks, nodeColor=<rgba 0.172 0 0.2344 1>}
  NamedRerouteDeclaration(None=@n35.Mask) {name=MouthMask, nodeColor=<rgba 1 0 0.7266 1>}
  n57 = NamedRerouteUsage()
  n58 = ScalarParameter() {defaultValue=4, group=Flipbook C, parameterName=C_GridSize, sortPriority=4}
  n59 = MaterialFunctionCall(
   OffsetScale=ScalarParameter() {defaultValue=0.65, group=Flipbook C, parameterName=C_OffsetScale, sortPriority=2}
   OffsetRotation=ScalarParameter() {group=Flipbook C, parameterName=C_OffsetRotation, sortPriority=2}
   OffsetU=Add(A=ScalarParameter() {group=Flipbook C, parameterName=C_OffsetU, sortPriority=2}, B=ComponentMask(None=@n57) {b=true}) {constB=0}
   OffsetV=ScalarParameter() {defaultValue=0.1, group=Flipbook C, parameterName=C_OffsetV, sortPriority=3}
   Number of Columns=@n58
   Index=ComponentMask(None=NamedRerouteUsage()) {r=true}
   Number of Rows=@n58
  ) {materialFunction=&MF_ChiveFlake_Keytar_Flipbook}
  NamedRerouteDeclaration(None=@n59.Coordinates) {name=EyesCoordinates, nodeColor=<rgba 1 0 0.7266 1>}
  NamedRerouteDeclaration(None=@n59.Mask) {name=EyesMask, nodeColor=<rgba 1 0 0.7266 1>}
  n60 = NamedRerouteUsage()
  n61 = TextureSample(UVs=@n60) {texture=/SparksCosmetics/Instrument/Keytar/ChiveFlake/Textures/T_ChiveFlake_Keytar_IdleTransition_Face_FB}
  n62 = Floor(None=Divide(A=NamedRerouteUsage()) {constB=16})
  n63 = LinearInterpolate(
   A=TextureSample(UVs=@n60) {texture=/SparksCosmetics/Instrument/Keytar/ChiveFlake/Textures/T_ChiveFlake_Keytar_LookAround_Mouth_FB}.RGBA
   B=TextureSample(UVs=@n60) {texture=/SparksCosmetics/Instrument/Keytar/ChiveFlake/Textures/T_ChiveFlake_Keytar_ChompMad_Mouth_FB}.RGBA
   Alpha=Saturate(None=Subtract(A=@n62))
  )
  n64 = Saturate(None=@n62)
  n65 = NamedRerouteUsage()
  n66 = TextureSampleParameter2D(UVs=@n65) {
   group=Flipbook A
   parameterName=A_Flipbook_Idle
   sortPriority=1
   texture=&T_ChiveFlake_Keytar_IdleTransition_Face_FB
  }
  n67 = TextureSampleParameter2D(UVs=@n65) {
   group=Flipbook A
   parameterName=A_Flipbook_Anim
   sortPriority=1
   texture=&T_ChiveFlake_Keytar_LookAround_Mouth_FB
  }
  n68 = NamedRerouteUsage()
  NamedRerouteDeclaration(None=StaticSwitchParameter(True=LinearInterpolate(A=@n61.RGBA, B=@n63, Alpha=@n64), False=LinearInterpolate(A=@n66.RGBA, B=@n67.RGBA, Alpha=@n68) {constAlpha=1}) {parameterName=ChoseTextureByIndex}) {name=MouthTexture, nodeColor=<rgba 0.2656 1 0 1>}
  n69 = NamedRerouteUsage()
  n70 = TextureSample(UVs=@n69) {texture=&T_ChiveFlake_Keytar_IdleTransition_Face_FB}
  n71 = Floor(None=Divide(A=NamedRerouteUsage()) {constB=16})
  n72 = LinearInterpolate(
   A=TextureSample(UVs=@n69) {texture=/SparksCosmetics/Instrument/Keytar/ChiveFlake/Textures/T_ChiveFlake_Keytar_LookAround_Eyes_FB}.RGBA
   B=TextureSample(UVs=@n69) {texture=/SparksCosmetics/Instrument/Keytar/ChiveFlake/Textures/T_ChiveFlake_Keytar_ChompMad_Eyes_FB}.RGBA
   Alpha=Saturate(None=Subtract(A=@n71))
  )
  n73 = Saturate(None=@n71)
  n74 = NamedRerouteUsage()
  n75 = TextureSampleParameter2D(UVs=@n74) {
   group=Flipbook C
   parameterName=C_Flipbook_Idle
   sortPriority=1
   texture=&T_ChiveFlake_Keytar_IdleTransition_Face_FB
  }
  n76 = TextureSampleParameter2D(UVs=@n74) {
   group=Flipbook C
   parameterName=C_Flipbook_Anim
   sortPriority=1
   texture=&T_ChiveFlake_Keytar_LookAround_Eyes_FB
  }
  n77 = NamedRerouteUsage()
  NamedRerouteDeclaration(None=StaticSwitchParameter(True=LinearInterpolate(A=@n70.RGBA, B=@n72, Alpha=@n73), False=LinearInterpolate(A=@n75.RGBA, B=@n76.RGBA, Alpha=@n77) {constAlpha=1}) {defaultValue=true, parameterName=ChoseTextureByIndex}) {name=EyesTexture, nodeColor=<rgba 0.2656 1 0 1>}
  n78 = ScalarParameter() {defaultValue=2, group=Flipbook B, parameterName=B_GridSize, sortPriority=4}
  n79 = MaterialFunctionCall(
   OffsetScale=ScalarParameter() {defaultValue=0.4, group=Flipbook B, parameterName=B_OffsetScale, sortPriority=2}
   OffsetRotation=ScalarParameter() {group=Flipbook B, parameterName=B_OffsetRotation, sortPriority=2}
   OffsetU=ScalarParameter() {group=Flipbook B, parameterName=B_OffsetU, sortPriority=2}
   OffsetV=ScalarParameter() {defaultValue=-0.3, group=Flipbook B, parameterName=B_OffsetV, sortPriority=3}
   Number of Columns=@n78
   Index=NamedRerouteUsage()
   Number of Rows=@n78
  ) {materialFunction=&MF_ChiveFlake_Keytar_Flipbook}
  n80 = TextureSampleParameter2D(UVs=@n79.Coordinates) {
   group=Flipbook B
   parameterName=B_Flipbook
   sortPriority=1
   texture=/SparksCosmetics/Instrument/Keytar/ChiveFlake/Textures/T_ChiveFlake_Keytar_Pulse_Star_FB
  }
  NamedRerouteDeclaration(None=@n80.RGB) {name=B_FlipbookColor, nodeColor=<rgba 0.0547 0 1 1>}
  NamedRerouteDeclaration(None=Multiply(A=@n80.A, B=@n79.Mask)) {name=B_FlipbookMask, nodeColor=<rgba 1 0.375 0 1>}
  n81 = NamedRerouteUsage()
  NamedRerouteDeclaration(None=ComponentMask(None=@n81) {b=true, g=true, r=true}) {name=C_FlipbookColor, nodeColor=<rgba 0.0547 0 1 1>}
  n82 = ComponentMask(None=@n81) {a=true}
  n83 = NamedRerouteUsage()
  NamedRerouteDeclaration(None=MaterialFunctionCall(In=Multiply(A=@n82, B=@n83), Power=Constant() {r=3}) {materialFunction=&SCurve}.Result) {name=C_FlipbookMask, nodeColor=<rgba 1 0.375 0 1>}
  n84 = Fmod(A=@n49, B=Constant() {r=1})
  n85 = ScalarParameter() {group=Flipbook General, parameterName=BreakerLoop, sliderMax=1}
  n86 = CurveAtlasRowParameter(CurveTime=Saturate(None=LinearInterpolate(A=@n49, B=@n84, Alpha=@n85))) {
   atlas=&CA_ChiveFlake_Keytar_Animations
   curve=&CLC_ChiveFlake_Keytar_Mad
   defaultValue=9
   group=Flipbook General
   parameterName=Breaker_CurveAtlas
  }
  NamedRerouteDeclaration(None=LinearInterpolate(A=@n86, B=NamedRerouteUsage(), Alpha=@n53) {constAlpha=1}) {name=AnimationIndex, nodeColor=<rgba 0.6406 0 1 1>}
  NamedRerouteDeclaration(None=@n46.R) {name=StarIndex, nodeColor=<rgba 1 0.6448 0 1>}
  NamedRerouteUsage()
  Round(None=ScalarParameter() {defaultValue=1, group=Flipbook General, parameterName=IsMusicPlaying})
  n87 = Time()
  n88 = ScalarParameter() {parameterName=GameTime}
  Saturate(None=Divide(A=Subtract(A=@n87, B=@n88)) {constB=5, desc=time it lasts})
  n89 = Subtract(A=Time() {period=1}, B=ScalarParameter() {group=Flipbook General, parameterName=PulseGameTime})
  n90 = Divide(B=ScalarParameter() {defaultValue=60, group=Flipbook General, parameterName=BPM}) {constA=60, constB=60}
  Saturate(None=Fmod(A=@n89, B=@n90))
  Saturate(None=@n89)
  Saturate(None=Subtract(B=ScalarParameter() {group=Flipbook General, parameterName=IgnoreEyesAnim, sliderMax=1}))
  ScalarParameter() {defaultValue=1, group=Flipbook General, parameterName=IsMusicPlaying}
  ScalarParameter() {group=Flipbook General, parameterName=IgnoreEyesAnim, sliderMax=1}
  n91 = Subtract(A=Time() {period=1}, B=ScalarParameter() {group=Flipbook General, parameterName=AltPulseGameTime})
  n92 = Multiply() {constB=2}
  n93 = ScalarParameter() {defaultValue=2, group=Flipbook General, parameterName=HighBPMDivision}
  n94 = Saturate(None=Fmod(A=@n91, B=Multiply(A=@n90, B=@n93) {constB=3}))
  n95 = ScalarParameter() {defaultValue=1, group=Flipbook General, parameterName=IsMusicPlaying}
  StaticSwitchParameter(
   True=Saturate(None=@n91)
   False=LinearInterpolate(B=StaticSwitchParameter(True=@n92, False=@n94) {group=Flipbook General, parameterName=UseMCP}, Alpha=Round(None=@n95)) {constB=0}
  ) {group=Flipbook General, parameterName=UsePulseEvent}
  StaticSwitchParameter() {group=Flipbook General, parameterName=UsePulseEvent}
  Step(Y=@n1) {constX=0.06}
 ]}
