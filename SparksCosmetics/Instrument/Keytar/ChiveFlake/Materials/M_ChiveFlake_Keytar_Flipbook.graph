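# Cloth-shaded anime material for the ChiveFlake keytar instrument cosmetic, built on material attributes for skeletal meshes.
# It chains the shared 2023 anime shader's Pre, Calculations and Post functions.
# MF_ChiveFlake_Keytar processes the Pre stage's lit color before Post, which also takes the shading gradients and surface inputs.
# Despite the Flipbook name, the SubUV function node shows no connections to that chain.
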Material /SparksCosmetics/Instrument/Keytar/ChiveFlake/Materials/M_ChiveFlake_Keytar_Flipbook
graph {
 properties {
  bAutomaticallySetUsageInEditor=false
  bUseMaterialAttributes=true
  bUsedWithMeshDeformer=true
  bUsedWithSkeletalMesh=true
  bUsedWithStaticLighting=true
  pixelDepthOffsetMode=PDOM_Legacy
  shadingModel=MSM_Cloth
  usedShadingModels=Cloth
 }
 nodes [
  n1 = MaterialFunctionCall!(CustomUV) {materialFunction=/Game/Global/Materials/Character_Parent_2023/Anime_Parent_2023/MaterialFunctions/MFC_AnimeShader_2023_Pre}
  n2 = MaterialFunctionCall!(CustomLightVector, OverrideNormal_WS, STT=@n1.STT, Normal_Map=@n1.STT, DFL=@n1.STT, FX=@n1.STT) {materialFunction=/Game/Global/Materials/Character_Parent_2023/Anime_Parent_2023/MaterialFunctions/MFC_AnimeShader_2023_Calculations}
  n3 = MaterialFunctionCall!(In_1=@n1.Color_Lit, In=@n1.Color_Lit) {materialFunction=&/MF_ChiveFlake_Keytar}
  MaterialFunctionCall!(
   Shading_Gradient=@n2.Shading_Gradient
   Threshold_Gradient=@n2.Shading_Gradient
   Specular_Gradient=@n2.Shading_Gradient
   Sheen_Mask=@n2.Shading_Gradient
   Thickness=@n2.Shading_Gradient
   DFL_Mask=@n2.Shading_Gradient
   FX=@n2.Shading_Gradient
   Metallic_Gradient=@n2.Shading_Gradient
   RimLight_Gradient=@n2.Shading_Gradient
   NormalBuffer=@n2.Shading_Gradient
   --------
   CINE_LightSources=@n2.Shading_Gradient
   CINE_LightColor=@n2.Shading_Gradient
   ---------
   Color_Lit=@n3.Result_1
   Color_Shaded=@n3.Result_1
   SRM=@n1.SRM
   Emissive=@n1.SRM
   DFL_Color=@n1.SRM
   Specular_Color=@n1.SRM
   RimLight_Color=@n1.SRM
  ) {materialFunction=/Game/Global/Materials/Character_Parent_2023/Anime_Parent_2023/MaterialFunctions/MFC_AnimeShader_2023_Post}
  MaterialFunctionCall!(Texture, UVs, SubImages, Frame) {materialFunction=/Engine/Functions/Engine_MaterialFunctions03/Texturing/SubUV_Function}
 ]
}
