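# Animation Blueprint for the ChiveFlake keytar cosmetic that animates its face through material parameters rather than skeletal poses.
# Named expressions swap eye and mouth flipbook textures, set curve-atlas positions and timestamp playback on a dynamic material instance.
# Timers drive blinks and idle glances, with mad, chomp and fake-beat showcases added in preview.
# Music beats trigger chomps on alternate beats, swings hold a mad face, and eliminations play a chomp sequence.
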AnimBlueprint /SparksCosmetics/Instrument/Keytar/ChiveFlake/Meshes/ChiveFlake_Keytar_AnimBP
target_skeleton=/SparksCharacterCommon/Instrument/Keytar/Generic/Meshes/Generic_Keytar_Skeleton
graphs [
 AnimGraph {nodes [Output Pose(Result=Input Pose())]}
 Blink {nodes [
   Blink() {is=title K2Node_FunctionEntry}
   > n1 = GetMaterialInstance!(execute, self) -> (then, MaterialInstance:Material Instance Dynamic Object Reference)
   > SetScalarParameterValue(self=@n1, ParameterName=BlinkGameTime, Value=GetGameTimeInSeconds())
  ]}
 BreakerDelay {nodes [
   n1 = Inputs!() -> (execute, BaseDelay:double=~, RandMin:double=4, RandMax:double=9)
   > Delay(Duration=float + float(A=@n1, B=RandomFloatInRange(Min=@n1.RandMin, Max=@n1.RandMax)))
   > n2 = Branch(Condition=IsAnimationPlaying!() -> (ReturnValue))
   > SetScalarParameterValue(self=Get MaterialInstance() -> (MaterialInstance:Material Instance Dynamic Object Reference), ParameterName=BreakerGameTime, Value=0)
   > Choose Breaker()
   Outputs!(then:Exec=@n2.else)
  ]}
 ChangeBreakerAtlasIndex {nodes [
   n1 = ChangeBreakerAtlasIndex!() -> (then, InCurve:Curve Linear Color Object Reference, ParameterName:Name=Breaker_CurveAtlas) {is=title K2Node_FunctionEntry}
   > n2 = GetCurvePosition!(
    execute
    self:Curve Linear Color Atlas Object Reference=Get Atlas() -> (Atlas:Curve Linear Color Atlas Object Reference)
    InCurve:Curve Linear Color Object Reference=@n1
   ) -> (then, Position:float, ReturnValue:bool)
   > n3 = Branch(Condition=@n2.ReturnValue)
   > GetMaterialInstance()
   n4 = Get MaterialInstance()
   > SetScalarParameterValue(self=@n4, ParameterName=@n1.ParameterName, Value=@n2)
   > SetScalarParameterValue(self=@n4, ParameterName=BreakerGameTime, Value=GetGameTimeInSeconds())
   > Return Node!(
    execute
    End Time:float=GetLinearColorValue!(self:Curve Linear Color Object Reference=@n1, InTime:float=0) -> (ReturnValue_R:float=~, ReturnValue_G:float=~, ReturnValue_B:float=~, ReturnValue_A:float=~).ReturnValue_A
    ReturnValue:bool=@n2.ReturnValue
   )
   SetScalarParameterValue(execute=@n3.else, self=Get MaterialInstance(), ParameterName=BreakerGameTime, Value=0)
   > Return Node(End Time=0.2, ReturnValue=@n2.ReturnValue)
  ]}
 ChangeIdleAtlasIndex {nodes [
   n1 = ChangeIdleAtlasIndex!() -> (then, InCurve:Curve Linear Color Object Reference) {is=title K2Node_FunctionEntry}
   > n2 = GetCurvePosition(self=Get Atlas(), InCurve=@n1)
   > n3 = GetMaterialInstance()
   > SetScalarParameterValue(self=@n3, ParameterName=Idle_CurveAtlas, Value=@n2)
   > Return Node!(execute)
  ]}
 EventGraph {nodes [
   n1 = Beat!() -> (OutputDelegate, then, BeatNumber:int, BeatInBar:int) {is=title K2Node_CustomEvent}
   n2 = "To Float (Integer)"(InInt=@n1.BeatInBar)
   > Branch(Condition="Equal (Float)"(A="% (Float)"!(A:double=@n2, B:double=2) -> (ReturnValue:double), B=1))
   > n3 = Branch(Condition=OR Boolean!(
     A=Get IsGettingKill() -> (IsGettingKill:bool)
     B=Get IsSwinging() -> (IsSwinging:bool)
     C:bool=Get IsStoppingSwing() -> (IsStoppingSwing:bool)
    ) -> (ReturnValue))
   Play Breaker Animation!(
    execute:Exec=@n3.else
    Key:"Name (by ref)"=MakeLiteralName(Value=BeatChomp)
    Freeze:bool
    Loop:bool
    UseBeat:bool
    IgnoreEyeAnim:bool
    ChangeIdleIndex:int=-1
   ) -> (then:Exec, End Time:float=~) {is=title K2Node_MacroInstance}
   PlayStopSwinging() {is=title K2Node_CustomEvent}
   > Branch(Condition=Get IsSwinging())
   > Set IsSwinging(IsSwinging:bool=false) -> (Output_Get:bool)
   > Set IsStoppingSwing(IsStoppingSwing:bool=true) -> (Output_Get:bool)
   > n4 = Play Breaker Animation(Key=MakeLiteralName(Value=MadOut), ChangeIdleIndex=-1) -> (End Time=~)
   > Delay(Duration=@n4)
   > Set IsStoppingSwing(IsStoppingSwing=false)
   n5 = Start Blinking() {is=title K2Node_CustomEvent}
   Delay(execute=[@n5.then, @n6.then], Duration=RandomFloatInRange(Min=3.7, Max=6))
   > n6 = Blink() {is=title K2Node_CallFunction}
   StartSwinging() {is=title K2Node_CustomEvent}
   > n7 = Branch(Condition=Get IsGettingKill())
   n8 = Branch(execute=@n7.else, Condition=Get IsSwinging())
   RetriggerableDelay(execute=[@n10.then, @n8.then], Duration=0.8)
   > n9 = Branch(Condition=Get IsGettingKill())
   PlayStopSwinging(execute=@n9.else) {is=title K2Node_CallFunction}
   n10 = PlaySwinging(execute=@n8.else)
   Choose Breaker() {is=title K2Node_CustomEvent}
   > n11 = Branch(Condition=IsAnimationPlaying())
   > Delay(Duration=5)
   > Choose Breaker() {is=title K2Node_CallFunction}
   n12 = Branch(execute=@n11.else, Condition=IsCosmeticObjectDisplayedAsPreview!(CosmeticObject:Object Reference=GetOwningActor()) -> (ReturnValue:bool))
   > FE Breaker()
   Play Breaker(execute=@n12.else)
   PlaySwinging() {is=title K2Node_CustomEvent}
   > Set IsSwinging(IsSwinging=true)
   > n13 = Play Breaker Animation(Key=MakeLiteralName(Value=MadIn), Freeze=true, ChangeIdleIndex=-1) -> (End Time=~)
   > Delay(Duration=@n13)
   > Play Breaker Animation(Key=MakeLiteralName(Value=Mad), Loop=true, ChangeIdleIndex=-1) -> (End Time=~)
   FE Breaker() {is=title K2Node_CustomEvent}
   > n14 = MultiGate!(execute:Exec, Reset:Exec, IsRandom:bool, Loop:bool=true, StartIndex:int=-1) -> (Out 0:Exec, Out 1:Exec, Out 2:Exec) {is=title K2Node_MultiGate}
   > n15 = Play Breaker Animation(Key=MakeLiteralName(Value=LookLeft), ChangeIdleIndex=-1) -> (End Time=~)
   > Breaker Delay!(execute:Exec, BaseDelay:double=@n15, RandMin:double=4, RandMax:double=8) -> (then:Exec) {is=title K2Node_MacroInstance}
   > Choose Breaker()
   n16 = Play Breaker Animation(execute=@n14.Out 1, Key=MakeLiteralName(Value=LookRight), ChangeIdleIndex=-1) -> (End Time=~)
   > Breaker Delay(BaseDelay=@n16, RandMin=4, RandMax=8)
   > Choose Breaker()
   n17 = MultiGate(execute=@n14.Out 2, Loop=true, StartIndex=-1)
   > n18 = Set RandomIntSequence(RandomIntSequence:int="Get (a copy)"(Array:Array of Integers=Get RandomSequence() -> (RandomSequence:Array of Integers)) -> (Output:int)) -> (Output_Get:int=0)
   n19 = Switch on Int!(execute:Exec=[@n21.then, @n18.then, @n20.then], Selection:int=Get RandomIntSequence() -> (RandomIntSequence:int=0)) -> (Default:Exec, 0:Exec, 1:Exec, 2:Exec)
   Play Mad!(execute:Exec=@n19.0) -> (then:Exec) {is=title K2Node_MacroInstance}
   > Breaker Delay(RandMin=4, RandMax=8)
   > Choose Breaker()
   Play Chomp!(execute:Exec=@n19.1) -> (then:Exec) {is=title K2Node_MacroInstance}
   > Breaker Delay(RandMin=4, RandMax=8)
   > Choose Breaker()
   Play Fake Beat!(execute:Exec=@n19.2) -> (then:Exec) {is=title K2Node_MacroInstance}
   > Breaker Delay(RandMin=4, RandMax=8)
   > Choose Breaker()
   n20 = Set RandomIntSequence(execute=@n17.Out 1, RandomIntSequence="Get (a copy)"(Array=Get RandomSequence(), Dimension 1=1)) -> (Output_Get=0)
   n21 = Set RandomIntSequence(execute=@n17.Out 2, RandomIntSequence="Get (a copy)"(Array=Get RandomSequence(), Dimension 1=2)) -> (Output_Get=0)
   Play Breaker() {is=title K2Node_CustomEvent}
   > Set FaceKey(FaceKey:Name=Random Array Item!(TargetArray:Array of Names=Make Array!("[0]":Name=LookRight, "[1]":Name=LookLeft) -> (Array:Array of Names)) -> (OutItem:Name, OutIndex:int=0) {is=title K2Node_CallArrayFunction}) -> (Output_Get:Name=None)
   > n22 = Play Breaker Animation(Key=Get FaceKey() -> (FaceKey:Name=None), ChangeIdleIndex=-1) -> (End Time=~)
   > Breaker Delay(BaseDelay=@n22, RandMin=4, RandMax=8)
   > Choose Breaker()
   Event BlueprintBeginPlay() {is=title K2Node_Event}
   > Shuffle!(execute, TargetArray:Array of Integers=Get RandomSequence()) -> (then) {is=title K2Node_CallArrayFunction}
   > Bind Event to On Highest Priority Clock Beat(
    self:Music Event Subsystem Object Reference=Get MusicEventSubsystem() -> (ReturnValue:Music Event Subsystem Object Reference) {is=title K2Node_GetSubsystem}
    Delegate=Create Event()
   ) {is=title K2Node_AddDelegate}
   > n23 = Cast To PlayerPawn_Athena(Object=TryGetPawnOwner())
   n24 = Create Event()
   > Bind Event to Swing Left(self:Player Pawn Athena Object Reference=@n23, Delegate=@n24) {is=title K2Node_AddDelegate}
   > Bind Event to Swing Right(self:Player Pawn Athena Object Reference=@n23, Delegate=@n24) {is=title K2Node_AddDelegate}
   Event BlueprintInitializeAnimation()
   > Delay(Duration=0.1)
   > n25 = Set Mesh(Mesh:Skeletal Mesh Component Object Reference=GetOwningComponent()) -> (Output_Get:Skeletal Mesh Component Object Reference)
   > n26 = Get Dynamic Material Instance(SkeletalMeshComponent=@n25)
   > Set MaterialInstance(MaterialInstance:Material Instance Dynamic Object Reference=@n26) -> (Output_Get:Material Instance Dynamic Object Reference)
   > n27 = Delay(Duration=0.2)
   n28 = MultiGate!(execute=[@n27.then, @n29.then], Reset, IsRandom, Loop, StartIndex=-1) -> (Out 0, Out 1, Out 2, Out 3:Exec)
   SetScalarParameterValue(execute=[@n30.then, @n28.Out 0, @n28.Out 1], self=Get MaterialInstance(), ParameterName=BlinkGameTime, Value=GetGameTimeInSeconds())
   > n29 = Delay(Duration=0.2)
   n30 = Delay(execute=@n28.Out 2, Duration=0.2)
   Start Blinking(execute=@n28.Out 3) {is=title K2Node_CallFunction}
   > Delay(Duration=3)
   > Choose Breaker()
   n31 = Event BlueprintProcessAnimInputEvent()
   n32 = Get IsGettingKill()
   > Branch(Condition=AND Boolean(A=IsNamed(self=@n31.AnimInputEvent, InEventName=aie.EnemyPlayerEliminated), B=NOT Boolean(A=@n32)))
   > Set IsGettingKill(IsGettingKill:bool=true) -> (Output_Get:bool)
   > PlayStopSwinging()
   > Delay(Duration=0.3)
   > Play Chomp()
   > Delay(Duration=0.7)
   > Set IsGettingKill(IsGettingKill=false)
   n33 = MultiGate!(execute, Reset, IsRandom=true, Loop=true, StartIndex=-1) -> (Out 0, Out 1)
   > Play Chomp()
   Play Mad(execute=@n33.Out 1)
  ]}
 GetCurveIndexValue {nodes [
   n1 = GetCurveIndexValue!() -> (then, Key:Name, InTime:float=~) {is=title K2Node_FunctionEntry}
   n2 = Find(
    TargetMap:Map of Names to Curve Float Object References=Get Curves() -> (Curves:Map of Names to Curve Float Object References)
    Key:"Name (by ref)"=@n1
   ) -> (Value:Curve Float Object Reference)
   > n3 = Branch(Condition=AND Boolean(A=@n2.ReturnValue, B=IsValid(Object=@n2)))
   > Return Node!(
    execute
    FlipbookIndex:int=Truncate(A=GetFloatValue!(self:Curve Float Object Reference=@n2, InTime:float=@n1.InTime) -> (ReturnValue:float))
    Found:bool=true
    MaxTime:double=GetTimeRange!(self:Curve Base Object Reference=@n2) -> (MinTime:float, MaxTime:float).MaxTime
   )
   Return Node(execute=@n3.else, FlipbookIndex=0, Found=false, MaxTime=0)
  ]}
 GetCurveLenght {nodes [
   n1 = GetCurveLenght!() -> (then, Key:Name) {is=title K2Node_FunctionEntry}
   n2 = Find(TargetMap=Get Curves(), Key=@n1)
   > n3 = Branch(Condition=AND Boolean(A=@n2.ReturnValue, B=IsValid(Object=@n2)))
   > Return Node!(execute, MaxTime=GetTimeRange(self=@n2).MaxTime, Found=true)
   Return Node(execute=@n3.else, MaxTime=0, Found=false)
  ]}
 GetMaterialInstance {nodes [
   GetMaterialInstance() {is=title K2Node_FunctionEntry}
   > n1 = Get Dynamic Material Instance(SkeletalMeshComponent=Get Mesh!() -> (Mesh))
   > n2 = Set MaterialInstance(MaterialInstance=@n1)
   > Return Node!(execute, MaterialInstance:Material Instance Dynamic Object Reference=@n2)
  ]}
 IsAnimationPlaying {nodes [
   IsAnimationPlaying() {is=title K2Node_FunctionEntry}
   n1 = GetSongPos!(
    self:Music Clock Object Reference=GetHighestPriorityClock!(self:Music Event Subsystem Object Reference=Get MusicEventSubsystem()) -> (ReturnValue:Music Clock Object Reference)
    Timebase:ECalibratedMusicTimebase Enum=VideoRenderTime
   ) -> (ReturnValue:"MIDI Song Position Structure (by ref)")
   > Return Node!(
    execute
    ReturnValue:bool=OR Boolean!(
     A=Is MIDI Song Position Valid!(SongPos:"MIDI Song Position Structure (by ref)"=@n1) -> (ReturnValue:bool)
     B=Get IsSwinging()
     C=Get IsStoppingSwing()
     D:bool=Get IsGettingKill()
    ) -> (ReturnValue)
   )
  ]}
 PlayBreakerAnimation {nodes [
   n1 = Inputs!() -> (execute, Key:"Name (by ref)", Freeze:bool=~, Loop:bool=~, UseBeat:bool=~, IgnoreEyeAnim:bool=~, ChangeIdleIndex:int=-1)
   n2 = Get MaterialInstance()
   SetScalarParameterValue(self=@n2, ParameterName=BreakerUsePulse, Value="To Float (Boolean)"(InBool=@n1.UseBeat))
   > n3 = SetScalarParameterValue(self=@n2, ParameterName=IgnoreEyesAnim, Value="To Float (Boolean)"(InBool=@n1.IgnoreEyeAnim))
   n4 = Find(
    TargetMap:Map of Names to Curve Linear Color Object References=Get Animations() -> (Animations:Map of Names to Curve Linear Color Object References)
    Key=@n1
   ) -> (Value:Curve Linear Color Object Reference)
   n5 = ChangeBreakerAtlasIndex!(execute=[@n8.then, @n7.else], self, InCurve:Curve Linear Color Object Reference=@n4, ParameterName:Name=Breaker_CurveAtlas) -> (then, End Time:float, ReturnValue:bool) {is=title K2Node_CallFunction}
   Outputs!(then=[@n3.then, @n11.then], End Time:float=@n5)
   n6 = Select!(
    Option 0:Curve Linear Color Object Reference=/SparksCosmetics/Instrument/Keytar/ChiveFlake/Animation/CLC_ChiveFlake_Keytar_Idle_01
    Option 1:Curve Linear Color Object Reference=/SparksCosmetics/Instrument/Keytar/ChiveFlake/Animation/CLC_ChiveFlake_Keytar_Idle_02
    Option 2:Curve Linear Color Object Reference=/SparksCosmetics/Instrument/Keytar/ChiveFlake/Animation/CLC_ChiveFlake_Keytar_Idle_03
    Option 3:Curve Linear Color Object Reference=/SparksCosmetics/Instrument/Keytar/ChiveFlake/Animation/CLC_ChiveFlake_Keytar_Idle_04
    Index:int=@n1.ChangeIdleIndex
   ) -> (ReturnValue:Curve Linear Color Object Reference)
   ChangeIdleAtlasIndex!(execute=@n12.then, self, InCurve:Curve Linear Color Object Reference=@n6) -> (then) {is=title K2Node_CallFunction}
   GetMaterialInstance(execute=@n14.then_0) {is=title K2Node_CallFunction}
   > n7 = Branch(Condition=@n4.ReturnValue)
   > n8 = SetBreakerTextures!(execute, self, Key:"Name (by ref)"=@n1) -> (then)
   n9 = "To Float (Boolean)"(InBool=@n1.Loop)
   n10 = "To Float (Boolean)"(InBool=@n1.Freeze)
   n11 = SetScalarParameterValue(execute=@n13.then, self=@n2, ParameterName=BreakerLoop, Value=@n9)
   n12 = Delay(execute=@n15.Is Valid, Duration=0.2)
   n13 = SetScalarParameterValue(execute=@n5.then, self=@n2, ParameterName=BreakerFreeze, Value=@n10)
   n14 = Sequence!(execute=@n1.execute) -> (then_0, then_1, then_2:Exec)
   n15 = Is Valid(exec=@n14.then_1, InputObject=@n6)
  ]}
 PlayChomp {nodes [
   Inputs!() -> (execute)
   > n1 = Play Breaker Animation(Key=MakeLiteralName(Value=ChompIn), Freeze=true, ChangeIdleIndex=-1) -> (End Time=~)
   > Delay(Duration=@n1)
   > n2 = Play Breaker Animation(Key=MakeLiteralName(Value=ChompSingle), Freeze=true, ChangeIdleIndex=-1) -> (End Time=~)
   > Delay(Duration=@n2)
   > n3 = Play Breaker Animation(Key=MakeLiteralName(Value=ChompSingle), Freeze=true, ChangeIdleIndex=-1) -> (End Time=~)
   > Delay(Duration=@n3)
   > n4 = Play Breaker Animation(Key=MakeLiteralName(Value=ChompMaintained), Freeze=true, ChangeIdleIndex=-1) -> (End Time=~)
   > Delay(Duration=@n4)
   > n5 = Play Breaker Animation(Key=MakeLiteralName(Value=ChompOut), ChangeIdleIndex=-1) -> (End Time=~)
   > Delay(Duration=@n5)
   > Outputs!(then)
  ]}
 PlayFakeBeat {nodes [
   Inputs()
   > GetMaterialInstance()
   > Play Breaker Animation(ChangeIdleIndex=2) -> (End Time=~)
   > Delay(Duration=0.1)
   > SetScalarParameterValue(self=Get MaterialInstance(), ParameterName=PulseGameTime, Value=GetGameTimeInSeconds())
   n1 = Make Literal Float(Value=0.15)
   > Delay(Duration=@n1)
   > Beat!(execute, self, BeatNumber:int=0, BeatInBar:int=0) -> (then) {is=title K2Node_CallFunction}
   n2 = Make Literal Float(Value=0.5)
   > Breaker Delay(RandMin=@n2, RandMax=@n2)
   > SetScalarParameterValue(self=Get MaterialInstance(), ParameterName=PulseGameTime, Value=GetGameTimeInSeconds())
   > Delay(Duration=@n1)
   > Beat(BeatNumber=0, BeatInBar=1)
   > Breaker Delay(RandMin=@n2, RandMax=@n2)
   > SetScalarParameterValue(self=Get MaterialInstance(), ParameterName=PulseGameTime, Value=GetGameTimeInSeconds())
   > Delay(Duration=@n1)
   > Beat(BeatNumber=0, BeatInBar=0)
   > Breaker Delay(RandMin=@n2, RandMax=@n2)
   > SetScalarParameterValue(self=Get MaterialInstance(), ParameterName=PulseGameTime, Value=GetGameTimeInSeconds())
   > Delay(Duration=@n1)
   > Beat(BeatNumber=1, BeatInBar=1)
   > Breaker Delay(RandMin=@n2, RandMax=@n2)
   > SetScalarParameterValue(self=Get MaterialInstance(), ParameterName=PulseGameTime, Value=GetGameTimeInSeconds())
   > Delay(Duration=@n1)
   > Beat(BeatNumber=0, BeatInBar=0)
   > Breaker Delay(RandMin=@n2, RandMax=@n2)
   > SetScalarParameterValue(self=Get MaterialInstance(), ParameterName=PulseGameTime, Value=GetGameTimeInSeconds())
   > Delay(Duration=@n1)
   > Beat(BeatNumber=0, BeatInBar=1)
   > Breaker Delay(RandMin=@n2, RandMax=@n2)
   > SetScalarParameterValue(self=Get MaterialInstance(), ParameterName=PulseGameTime, Value=GetGameTimeInSeconds())
   > Delay(Duration=@n1)
   > Beat(BeatNumber=0, BeatInBar=0)
   > Breaker Delay(RandMin=@n2, RandMax=@n2)
   > SetScalarParameterValue(self=Get MaterialInstance(), ParameterName=PulseGameTime, Value=GetGameTimeInSeconds())
   > Delay(Duration=@n1)
   > Beat(BeatNumber=0, BeatInBar=1)
   > Breaker Delay(RandMin=@n2, RandMax=@n2)
   > SetScalarParameterValue(self=Get MaterialInstance(), ParameterName=PulseGameTime, Value=GetGameTimeInSeconds())
   > Delay(Duration=@n1)
   > Beat(BeatNumber=0, BeatInBar=0)
   > Breaker Delay(RandMin=@n2, RandMax=@n2)
   > SetScalarParameterValue(self=Get MaterialInstance(), ParameterName=PulseGameTime, Value=GetGameTimeInSeconds())
   > Play Breaker Animation(Key=MakeLiteralName(Value=blink), ChangeIdleIndex=0) -> (End Time=~)
   > Outputs()
  ]}
 PlayMad {nodes [
   Inputs()
   > n1 = Play Breaker Animation(Key=MakeLiteralName(Value=MadIn), Freeze=true, ChangeIdleIndex=-1) -> (End Time=~)
   > Delay(Duration=@n1)
   > Play Breaker Animation(Key=MakeLiteralName(Value=Mad), Freeze=true, Loop=true, ChangeIdleIndex=-1) -> (End Time=~)
   > Delay(Duration=1.5)
   > n2 = Play Breaker Animation(Key=MakeLiteralName(Value=MadOut), ChangeIdleIndex=-1) -> (End Time=~)
   > Delay(Duration=@n2)
   > Outputs()
  ]}
 PlayMultiAnim {nodes [
   n1 = Inputs!() -> (Play:Exec, Stop:Exec, Key:Array of Names)
   n2 = Local Boolean!() -> (Variable:bool=~) {is=title K2Node_TemporaryVariable}
   > Assign!(execute:Exec, Variable:bool=@n2, Value:bool=true) -> (then:Exec) {is=title K2Node_AssignmentStatement}
   n3 = Local Integer!() -> (Variable:int) {is=title K2Node_TemporaryVariable}
   > n4 = Assign(Variable:int=@n3, Value:int)
   n5 = Length(TargetArray:Array of Names=@n1)
   n6 = Branch(
    execute=[@n4.then, @n8.then]
    Condition=AND Boolean!(A="integer >= integer"(A=@n3), B="integer <= integer"(A=@n3, B=@n5), C:bool=@n2) -> (ReturnValue)
   )
   > n7 = Play Breaker Animation(
    Key="Get (a copy)"(Array:Array of Names=@n1, Dimension 1=@n3) -> (Output:Name)
    Freeze="Not Equal (Integer)"(A=@n3, B=int - int(A=@n5))
    ChangeIdleIndex=-1
   ) -> (End Time=~)
   > Assign(Variable=@n3, Value=int + int(A=@n3))
   > n8 = Delay(Duration=@n7)
   Outputs!(AnimEnd:Exec=@n6.else)
   Assign(execute=@n1.Stop, Variable:bool=@n2, Value:bool)
  ]}
 SetBreakerTextures {nodes [
   n1 = SetBreakerTextures!() -> (then, Key:"Name (by ref)") {is=title K2Node_FunctionEntry}
   n2 = Find(TargetMap:Map of Names to Integers=Get AnimationTextureIndex() -> (AnimationTextureIndex:Map of Names to Integers), Key=@n1) -> (Value:int)
   > SetTextureParameterValue(
    self=Get MaterialInstance()
    ParameterName=A_Flipbook_Anim
    Value="Get (a copy)"(
     Array:Array of Texture Object References=Make Array!(
      "[0]":Texture Object Reference=/SparksCosmetics/Instrument/Keytar/ChiveFlake/Textures/T_ChiveFlake_Keytar_IdleTransition_Face_FB
      "[1]":Texture Object Reference=/SparksCosmetics/Instrument/Keytar/ChiveFlake/Textures/T_ChiveFlake_Keytar_LookAround_Mouth_FB
      "[2]":Texture Object Reference=/SparksCosmetics/Instrument/Keytar/ChiveFlake/Textures/T_ChiveFlake_Keytar_ChompMad_Mouth_FB
     ) -> (Array:Array of Texture Object References)
     Dimension 1=@n2
    ) -> (Output:Texture Object Reference)
   )
   > SetTextureParameterValue(
    self=Get MaterialInstance()
    ParameterName=C_Flipbook_Anim
    Value="Get (a copy)"(
     Array=Make Array(
      "[0]"=&T_ChiveFlake_Keytar_IdleTransition_Face_FB
      "[1]"=/SparksCosmetics/Instrument/Keytar/ChiveFlake/Textures/T_ChiveFlake_Keytar_LookAround_Eyes_FB
      "[2]"=/SparksCosmetics/Instrument/Keytar/ChiveFlake/Textures/T_ChiveFlake_Keytar_ChompMad_Eyes_FB
     )
     Dimension 1=@n2
    )
   )
  ]}
 UpdateBPM {nodes [
   n1 = UpdateBPM!() -> (then, Value:float=60) {is=title K2Node_FunctionEntry}
   > n2 = GetMaterialInstance()
   > SetScalarParameterValue(self=@n2, ParameterName=BPM, Value=@n1)
   > SetScalarParameterValue(self=@n2, ParameterName=PulseGameTime, Value=GetGameTimeInSeconds())
  ]}
 UpdateCurveTime_2 {nodes [
   n1 = Inputs!() -> (execute:Exec, SetCurve:Exec, Key:Name=none, TextureIndex:int, ResetTimer:Exec, ResetTime:double=~)
   n2 = "Local Float (double-precision)"!() -> (Variable:double=~) {is=title K2Node_TemporaryVariable}
   Assign(execute=@n1.ResetTimer, Variable:double=@n2, Value:double=@n1.ResetTime)
   n3 = GetCurveIndexValue!(Key:Name=@n1, InTime:float=@n2) -> (FlipbookIndex:int=0, Found:bool, MaxTime:double) {is=title K2Node_CallFunction}
   n4 = Branch(execute=@n1.execute, Condition=@n3.Found)
   > Assign(Variable=@n2, Value=float + float(A=@n2, B=GetDeltaSeconds!(self:Anim Instance Object Reference) -> (ReturnValue:float)))
   > n5 = Branch(Condition="float >= float"(A=@n2, B=@n3.MaxTime))
   > n6 = Sequence!(execute) -> (then_0, then_1)
   > n7 = Assign(Variable=@n2, Value=float - float(A=@n2, B=@n3.MaxTime))
   n8 = UpdateFlipbookPose!(execute=[@n7.then, @n5.else], self, Texture Index:int=@n1.TextureIndex, Flipbook Index:int=@n3) -> (then)
   Outputs!(then:Exec=[@n8.then, @n9.then_0], Variable:double=@n2, LoopEnd:Exec=[@n6.then_1, @n9.then_1])
   n9 = Sequence(execute=@n4.else)
  ]}
 UpdateFlipbookPose {nodes [
   n1 = UpdateFlipbookPose!() -> (then, Texture Index:int, Flipbook Index:int) {is=title K2Node_FunctionEntry}
   > n2 = GetMaterialInstance()
   n3 = Select!(Option 0:String=A_, Option 1:String=B_, Option 2:String=C_, Index=@n1) -> (ReturnValue:String)
   > SetScalarParameterValue(self=@n2, ParameterName=String To Name(InString=Append(A=@n3, B=Index)), Value="To Float (Integer)"(InInt=@n1.Flipbook Index))
  ]}
]
variables [
 Animations {
  type="map<CurveLinearColor>"
  default {
   beatChomp=/SparksCosmetics/Instrument/Keytar/ChiveFlake/Animation/CLC_ChiveFlake_Keytar_BeatChomp
   chomp=null
   chompIn=/SparksCosmetics/Instrument/Keytar/ChiveFlake/Animation/CLC_ChiveFlake_Keytar_ChompIn
   chompMaintained=/SparksCosmetics/Instrument/Keytar/ChiveFlake/Animation/CLC_ChiveFlake_Keytar_ChompMaintained
   chompOut=/SparksCosmetics/Instrument/Keytar/ChiveFlake/Animation/CLC_ChiveFlake_Keytar_ChompOut
   chompSingle=/SparksCosmetics/Instrument/Keytar/ChiveFlake/Animation/CLC_ChiveFlake_Keytar_ChompSingle
   lookLeft=/SparksCosmetics/Instrument/Keytar/ChiveFlake/Animation/CLC_ChiveFlake_Keytar_LookLeft
   lookRight=/SparksCosmetics/Instrument/Keytar/ChiveFlake/Animation/CLC_ChiveFlake_Keytar_LookRight
   mad=/SparksCosmetics/Instrument/Keytar/ChiveFlake/Animation/CLC_ChiveFlake_Keytar_Mad
   madIn=/SparksCosmetics/Instrument/Keytar/ChiveFlake/Animation/CLC_ChiveFlake_Keytar_IdleToMad
   madOut=/SparksCosmetics/Instrument/Keytar/ChiveFlake/Animation/CLC_ChiveFlake_Keytar_MadToIdle
   twoBlink=/SparksCosmetics/Instrument/Keytar/ChiveFlake/Animation/CLC_ChiveFlake_Keytar_TwoBlink
  }
 }
 AnimationTextureIndex {
  type="map<integer>"
  default {beatChomp=2, chomp=2, chompIn=2, chompMaintained=2, chompOut=2, chompSingle=2, lookLeft=1, lookRight=1, mad=2, madIn=2, madOut=2, twoBlink=0}
 }
 Atlas {type=CurveLinearColorAtlas, default=/SparksCosmetics/Instrument/Keytar/ChiveFlake/Animation/CA_ChiveFlake_Keytar_Animations}
 BPM {type=number, default=104.1709}
 CurentCurve {type=CurveLinearColor, default=&CLC_ChiveFlake_Keytar_LookLeft}
 CurrentAnimation {type="array<string>", default []}
 Curves {type="map<CurveFloat>", default {}}
 FaceKey {type=string, default=StarBeat}
 GeneratedBreakerCurveList {type="array<string>", default [LookRight,LookRight,LookRight,LookRight,LookLeft,LookLeft,LookLeft,LookLeft,Mad,Chomp]}
 IdleBreaker {type=boolean}
 IdleBreakerCurves {type="map<integer>", default {}}
 IdleCurves {
  type="map<CurveLinearColor>"
  default {
   closedMouth=&CLC_ChiveFlake_Keytar_Idle_03
   default=&CLC_ChiveFlake_Keytar_Idle_01
   la=&CLC_ChiveFlake_Keytar_Idle_02
   mad=&CLC_ChiveFlake_Keytar_Idle_04
  }
 }
 IdleFaceKey {type=string, default=StarBeat}
 IsGettingKill {type=boolean}
 IsMusicPlaying {type=boolean}
 IsStoppingSwing {type=boolean}
 IsSwinging {type=boolean}
 MaterialInstance {type=MaterialInstanceDynamic}
 Mesh {type=SkeletalMeshComponent}
 MusicState {type=EMusicClockState, default=Stopped}
 NewVar {type=TimerHandle}
 QueuedAnim {type=string, default=LookRight}
 RandomIntSequence {type=integer}
 RandomSequence {type="array<integer>", default [0,1,2]}
 Target {type=MusicClockComponent, default=null}
 TextureSwitch_Eyes {type="map<integer>", default {chomp=0}}
 TextureSwitch_Mouth {type="map<integer>", default {}}
 Time {type=number}
 UseBeatBreaker {type=boolean, default=true}
]
